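# Challenge stage indicator widget that selects text and an icon for either a boss or minion stage.
# Setup Stage Information caches both text and texture sets.
# UpdateStageText picks one by IsBossStage, passes it to the setters, then waits AnimationDelay before calling PlayIntroAnimations.
# Construct calls ResetStageText, but that function, both setters and the intro/outro events are empty here, and no widgets are present.
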WidgetBlueprint /AIProgressBar/UI/WBPs/WBP_Challenge_StageIndicatorText
graphs [
 EventGraph {nodes [
   n1 = UpdateStageText!() -> (OutputDelegate, then, IsBossStage:bool=~, AnimationDelay:double=~, WithDelayedOutroAnim:bool=~) {is=title K2Node_CustomEvent}
   > SetStageIcon!(
    execute
    self
    StageTexture:Texture Object Reference=Select(
     Option 0:Texture Object Reference=Get MinionTexture() -> (MinionTexture:Texture Object Reference)
     Option 1:Texture Object Reference=Get Boss Texture() -> (Boss Texture:Texture Object Reference)
     Index=@n1.IsBossStage
    ) -> (ReturnValue:Texture Object Reference)
   ) -> (then)
   > SetStageText!(
    execute
    self
    StageText:Text=SelectText(A=Get BossStageText() -> (BossStageText:Text), B=Get MinionsStageText() -> (MinionsStageText:Text), bPickA=@n1.IsBossStage)
   ) -> (then)
   > Delay(Duration=@n1.AnimationDelay)
   > PlayIntroAnimations!(execute, self, WithDelayedOutroAnim:bool=@n1.WithDelayedOutroAnim) -> (then)
   Event Construct()
   > ResetStageText!(execute, self, NeedsChallengeCompletedOutro:bool=false) -> (then)
   PlayIntroAnimations!() -> (OutputDelegate, then, WithDelayedOutroAnim:bool=~) {is=title K2Node_CustomEvent}
   PlayOutroAnimations() {is=title K2Node_CustomEvent}
  ]}
 ResetStageText {nodes [ResetStageText!() -> (then, NeedsChallengeCompletedOutro:bool=~) {is=title K2Node_FunctionEntry}]}
 Sequencer Events {links [], nodes []}
 SetStageIcon {nodes [SetStageIcon!() -> (then, StageTexture:Texture Object Reference) {is=title K2Node_FunctionEntry}]}
 SetStageText {nodes [SetStageText!() -> (then, StageText:Text) {is=title K2Node_FunctionEntry}]}
 Setup Stage Information {nodes [
   n1 = Setup Stage Information!() -> (then, BossStageText:Text, MinionsStageText:Text, BossTexture:Texture Object Reference, MinionTexture:Texture Object Reference) {is=title K2Node_FunctionEntry}
   > Set BossStageText(BossStageText:Text=@n1) -> (Output_Get:Text)
   > Set MinionsStageText(MinionsStageText:Text=@n1.MinionsStageText) -> (Output_Get:Text)
   > Set Boss Texture(Boss Texture:Texture Object Reference=@n1.BossTexture) -> (Output_Get:Texture Object Reference)
   > Set MinionTexture(MinionTexture:Texture Object Reference=@n1.MinionTexture) -> (Output_Get:Texture Object Reference)
  ]}
]
variables [
 Boss Texture {type=Texture, default=null}
 BossStageText {type=string, default=""}
 MinionsStageText {type=string, default=""}
 MinionTexture {type=Texture, default=null}
 TimeToOutro {type=number, default=3}
]
widgets {info {inheritedWidgetCount=0, parentClass=&UserWidget, rootWidgetClass=null, widgetCount=0}, widgets []}
