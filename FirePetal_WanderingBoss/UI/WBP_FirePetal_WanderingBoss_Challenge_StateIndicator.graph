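# FirePetal Wandering Boss challenge widget, derived from WBP_Challenge_StageIndicatorText, that shows uppercase stage text such as "Eliminate the Minions!".
# The intro animation reveals the hidden row, widening and fading in the text, then starts a TimeToOutro timer, apparently for the outro that fades and hides it.
# UpdateStageText defers to the parent, while SetStageText writes directly to the text block.
# ResetStageText stops animations and branches on challenge completion; SetStageIcon is empty.
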WidgetBlueprint /FirePetal_WanderingBoss/UI/WBP_FirePetal_WanderingBoss_Challenge_StateIndicator
animations [
 Anim_TextIntro {
  binding_count=3
  length_seconds=1.35
  object=&self:Anim_TextIntro
  sequence {bindings [
    Border_Text {
     id=#1
     possessed_class=Border
     tracks [{
       type=MovieSceneDoubleVectorTrack
       display_name=Desired Size Scale
       sections [{
         type=MovieSceneDoubleVectorSection
         end_frame=27
         start_frame=0
         channels [
          Vector.X {type=MovieSceneScriptingDoubleChannel, keys [[10.9997,0],[20,0],[26,1]]}
          Vector.Y {type=MovieSceneScriptingDoubleChannel, const [1,1]}
         ]
        }]
      }]
    }
    HorizontalBox_0 {
     id=#2
     possessed_class=HorizontalBox
     tracks [
      {
       type=MovieSceneFloatTrack
       display_name=Render Opacity
       sections [{type=MovieSceneFloatSection, end_frame=12, start_frame=0, channels [None {type=MovieSceneScriptingFloatChannel, const [1,1]}]}]
      }
      {
       type=MovieSceneByteTrack
       display_name=Visibility
       sections [{type=MovieSceneByteSection, end_frame=12, start_frame=0, channels [None {type=MovieSceneScriptingByteChannel, const [4,1]}]}]
      }
     ]
    }
    Text_StageIndicator {
     id=#3
     possessed_class=CommonTextBlock
     tracks [{
       type=MovieSceneFloatTrack
       display_name=Render Opacity
       sections [{
         type=MovieSceneFloatSection
         end_frame=27
         start_frame=0
         channels [None {type=MovieSceneScriptingFloatChannel, keys [[10.9997,0],[19.9997,0],[25.9997,1]]}]
        }]
      }]
    }
   ]}
 }
 Anim_TextOutro {
  binding_count=1
  length_seconds=0.5
  object=&self:Anim_TextOutro
  sequence {bindings [HorizontalBox_0 {
     id=#4
     possessed_class=HorizontalBox
     tracks [
      {
       type=MovieSceneFloatTrack
       display_name=Render Opacity
       sections [{type=MovieSceneFloatSection, end_frame=10, start_frame=0, channels [None {type=MovieSceneScriptingFloatChannel, keys [[0,1],[10,0]]}]}]
      }
      {
       type=MovieSceneByteTrack
       display_name=Visibility
       sections [{type=MovieSceneByteSection, end_frame=10, start_frame=10, channels [None {type=MovieSceneScriptingByteChannel, const [1,1]}]}]
      }
     ]
    }]}
 }
]
graphs [
 EventGraph {nodes [
   Event Construct()
   > Parent: Construct()
   n1 = Event UpdateStageText!() -> (OutputDelegate, then, IsBossStage:bool, AnimationDelay:double, WithDelayedOutroAnim:bool) {is=title K2Node_Event}
   > Parent: UpdateStageText!(execute, IsBossStage:bool=@n1.IsBossStage, AnimationDelay:double=@n1.AnimationDelay, WithDelayedOutroAnim:bool=@n1.WithDelayedOutroAnim) -> (then) {is=title K2Node_CallParentFunction}
   Event PlayIntroAnimations!() -> (OutputDelegate, then, WithDelayedOutroAnim:bool) {is=title K2Node_Event}
   > n2 = "Play Animation with Finished event (legacy)"!(
    execute:Exec
    Widget:User Widget Object Reference=Self-Reference()
    InAnimation:Widget Animation Object Reference=Get Anim_TextIntro() -> (Anim_TextIntro:Widget Animation Object Reference)
    StartAtTime:float=0
    NumLoopsToPlay:int=1
    PlayMode:EUMGSequencePlayMode Enum
    PlaybackSpeed:float=1
   ) -> (then:Exec, Result:UMGSequence Player Object Reference, Finished:Exec) {is=title K2Node_PlayAnimation}
   Set Timer by Event(execute=@n2.Finished, Delegate=Create Event(), Time=Get TimeToOutro() -> (TimeToOutro:float), bLooping=false)
   Event PlayOutroAnimations() {is=title K2Node_Event}
   > PlayAnimation(InAnimation=Get Anim_TextOutro() -> (Anim_TextOutro:Widget Animation Object Reference))
  ]}
 ResetStageText {nodes [
   n1 = ResetStageText!() -> (then, NeedsChallengeCompletedOutro:bool) {is=title K2Node_FunctionEntry}
   > StopAllAnimations(self:User Widget Object Reference)
   > n2 = Branch(Condition=@n1)
   > Set BossStageText(BossStageText:Text) -> (Output_Get:Text)
   "SetText (Text)"(execute=@n2.else, self=Get Text_StageIndicator() -> (Text_StageIndicator:Common Text Object Reference))
  ]}
 SetStageIcon {nodes [SetStageIcon!() -> (then, StageTexture:Texture Object Reference) {is=title K2Node_FunctionEntry}]}
 SetStageText {nodes [
   n1 = SetStageText!() -> (then, StageText:Text) {is=title K2Node_FunctionEntry}
   > "SetText (Text)"(self=Get Text_StageIndicator(), InText=@n1)
  ]}
]
variables []
widgets {
 info {
  inheritedWidgetCount=0
  parentClass=/AIProgressBar/UI/WBPs/WBP_Challenge_StageIndicatorText.WBP_Challenge_StageIndicatorText_C
  rootWidgetClass=&HorizontalBox
  widgetCount=3
 }
 widgets [
  HorizontalBox_0 {widgetClassPath=&HorizontalBox, widget=HorizontalBox_0, properties {visibility=Hidden}}
  Border_Text {
   widgetClassPath=&Border
   parent=HorizontalBox_0
   widget=Border_Text
   properties {brushColor=<rgba 1 1 1 0>}
   slot {class=HorizontalBoxSlot, properties {verticalAlignment=VAlign_Center}}
  }
  Text_StageIndicator {
   widgetClassPath=&CommonTextBlock
   bIsVariable=true
   parent=Border_Text
   widget=Text_StageIndicator
   properties {
    justification=Center
    text="Eliminate the Minions!"
    textTransformPolicy=ToUpper
    font {
     bForceMonospaced=false
     bMaterialIsStencil=false
     fontMaterial=null
     fontObject=&Burbank
     letterSpacing=36
     monospacedWidth=0.552
     size=24
     skewAmount=0
     typefaceFontName=Burbank-Black
     outlineSettings {
      bApplyOutlineToDropShadows=false
      bMiteredCorners=false
      bSeparateFillAlpha=false
      outlineMaterial=null
      outlineSize=0
      outlineColor=<rgba 0.0781 0.0781 0.0781 0.235>
     }
    }
    shadowColorAndOpacity=<rgba 0 0 0 0.56>
    shadowOffset=<1.5 1.5>
   }
   slot {class=BorderSlot}
  }
 ]
}
